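# Material function, exposed in the Fortnite Wind and Water library, that applies water interactions to incoming material attributes.
# Real-time water data supplies default height map and normal inputs, both overridable.
# The height map drives surface color modulation of base color and emissive and multiplies opacity, while its normal feeds a normal modification function alongside the ambient normal.
# A render-to-texture switch function outputs either the modified or untouched attributes.
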MaterialFunction /Figment_S06/FN_6_20/WaterAndWind/MaterialFunctions/Fortnite_WaterInteractions_BaseColorAndNormal
graph {
 properties {bExposeToLibrary=true, libraryCategoriesText [Fortnite Wind and Water]}
 nodes [
  n1 = FunctionInput(Preview=MakeMaterialAttributes!(
    BaseColor
    Metallic
    Specular
    Roughness
    Anisotropy
    EmissiveColor
    Opacity
    OpacityMask
    Normal
    Tangent
    WorldPositionOffset
    SubsurfaceColor
    ClearCoat
    ClearCoatRoughness
    AmbientOcclusion
    Refraction
    CustomizedUVs_0
    CustomizedUVs_1
    CustomizedUVs_2
    CustomizedUVs_3
    CustomizedUVs_4
    CustomizedUVs_5
    CustomizedUVs_6
    CustomizedUVs_7
    PixelDepthOffset
    ShadingModel
    Displacement
   )) {inputType=FunctionInput_MaterialAttributes}
  n2 = GetMaterialAttributes(None=@n1) {attributeGetTypes [
    #1
    #2
    #3
    #4
   ]}
  n3 = StaticBool() {value=true}
  n4 = MaterialFunctionCall!(Normal intensity, "World space offset to calculate normal ") {materialFunction=&/Fortnite_SampleRealTimeWaterData}
  n5 = FunctionInput(Preview=@n4.Height Map) {bUsePreviewValueAsDefault=true, inputName=Height Map, sortPriority=2}
  n6 = SetMaterialAttributes!(
   MaterialAttributes=@n1
   BaseColor=MaterialFunctionCall!(Diffuse=@n2.BaseColor, Use Player Interaction Base Color Modulation=@n3, HeightMap=@n5) {materialFunction=&/Fortnite_WaterInteractions_SurfaceColorModulation}.Result
   Normal=MaterialFunctionCall!(
    Water Interaction Normal=FunctionInput(Preview=@n4.Normal from height map) {bUsePreviewValueAsDefault=true, inputName=Normal From Height Map, sortPriority=1}
    Ambient Normal=@n2.Normal
    Use Player Interaction Normals
   ) {materialFunction=&/Fortnite_WaterInteractions_NormalModification}.Result
   EmissiveColor=MaterialFunctionCall(Diffuse=@n2.EmissiveColor, Use Player Interaction Base Color Modulation=@n3, HeightMap=@n5) {materialFunction=&Fortnite_WaterInteractions_SurfaceColorModulation}.Result
   Opacity=Multiply(A=@n2.Opacity, B=@n5)
  ) {attributeSetTypes [
    #1
    #2
    #3
    #4
   ]}
  FunctionOutput(None=MaterialFunctionCall!(
    With RTT Features
    WithOut RTT Features
    With RTT Features V2
    WithOut RTT Features V2
    With RTT Features Att=@n6
    WithOut RTT Features Att=@n1
   ) {materialFunction=/Figment_S06/FN_6_20/WaterAndWind/EnableOrDisableWind/Fortnite_RenderToTextureEnabled}.Att)
 ]
}
