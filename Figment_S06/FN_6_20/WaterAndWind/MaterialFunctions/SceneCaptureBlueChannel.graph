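# SceneCaptureBlueChannel is a material function in the Fortnite Wind and Water library that packs a flow direction and world height into one vector.
# It takes a two-component "WS Flow Map Dir" input and appends the Z component of World Position as the third channel.
# The name suggests the result feeds a scene capture, with height stored in blue beside flow in red and green.
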MaterialFunction /Figment_S06/FN_6_20/WaterAndWind/MaterialFunctions/SceneCaptureBlueChannel
graph {
 properties {libraryCategoriesText [Fortnite Wind and Water]}
 nodes [
  n1 = WorldPosition()
  FunctionOutput(None=AppendVector(
    A=FunctionInput() {bUsePreviewValueAsDefault=true, inputName=WS Flow Map Dir, inputType=FunctionInput_Vector2}
    B=MaterialFunctionCall(Float3=@n1.XYZ) {materialFunction=&BreakOutFloat3Components}.B
   ))
 ]
}
